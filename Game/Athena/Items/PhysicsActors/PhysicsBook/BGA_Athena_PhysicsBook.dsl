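; BGA_Athena_PhysicsBook is the actor Blueprint for a physics book item in Athena that cannot be killed by damage.
; It overrides ShouldDie to always return false, so the book never dies regardless of damage, instigator or causer.
; BeginPlay and the construction script only call their parent versions, while the overlap and tick events are empty.
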(event EventBeginPlay
  (|CallParentFunction))

(event Collision|EventActorBeginOverlap
  (params (OtherActor :type Actor%20Object%20Reference)))

(event EventTick
  (params (DeltaSeconds :type Float%20%28single-precision%29)))

(fn ShouldDie
  (params (Damage :type Float%20%28single-precision%29) (EventInstigator :type Controller%20Object%20Reference) (DamageCauser :type Actor%20Object%20Reference) (bIsTest :type Boolean) (EffectContext :type Gameplay%20Effect%20Context%20Handle%20Structure%20%28by%20ref%29))
  ;[1 1 1] Make sure the book won't die
  (return false))

(fn ConstructionScript
  (|CallParentFunction))
